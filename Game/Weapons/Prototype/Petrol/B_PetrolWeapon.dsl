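; Prototype petrol can weapon that pours a fuel trail, registering its projectiles as splatter sources.
; A replicated held pickup actor attaches with collision disabled; unequipping or holstering it undropped flags it to explode, then hides and destroys it.
; Equip and holster events manage leak Niagara and server ammo mod; fire FX drive the leak, pour audio and sound indicator.
; Windup tag tasks toggle targeting HUD key actions.
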(event Custom|BGADestroyedWhileHeld
  ;[1 1 1] On BGA Shot out of hands
  (CallFunction|UnbindHolsterEvents))

(event Custom|BindHolsterEvents
  ;[1 1 1] Holster Event binding / unbinding
  (bind _asfort_pawn (Utilities|Casting|CastToFortPawn (Game|GetInstigator)))
  (EventDispatchers|AssignOnHolsteredEvent _asfort_pawn (EventDispatchers|CreateEvent :!event Custom|OnHolstered))
  (EventDispatchers|AssignOnUnholsteredEvent _asfort_pawn (EventDispatchers|CreateEvent :!event Custom|OnUnholstered)))

(event Custom|UnbindHolsterEvents
  ;[1 1 1] Holster Event binding / unbinding
  (bind _asfort_pawn (Utilities|Casting|CastToFortPawn (Game|GetInstigator)))
  (EventDispatchers|UnbindEventfromOnHolsteredEvent _asfort_pawn (EventDispatchers|CreateEvent :!event Custom|OnHolstered))
  (EventDispatchers|UnbindEventfromOnUnholsteredEvent _asfort_pawn (EventDispatchers|CreateEvent :!event Custom|OnUnholstered)))

(event EventBeginPlay
  ;[1 1 1] On begin play set up HUD key actions
  (|CallParentFunction)
  (if (Variables|Default|GetShowHUDKeyActions)
    (WeaponHUDData|SetHudKeyActionsVisibility (Utilities|Array|MakeArray (Utilities|Struct|MakeWeaponHudKeyActionVisibility 0 true) (Utilities|Struct|MakeWeaponHudKeyActionVisibility 1 true)))
    (Ability|Tasks|WaitGameplayTagAddToActor (Game|GetInstigator) (Variables|Default|GetTag_InWindup)
      (:then
        (Ability|Tasks|WaitGameplayTagRemoveFromActor (Game|GetInstigator) (Variables|Default|GetTag_InWindup)
          (:Removed
            (CallFunction|HUDKeyActionTargeting false))))
      (:Added
        (CallFunction|HUDKeyActionTargeting true)))))

(event Custom|HUDKeyAction_Targeting
  (params (IsTargeting :type Boolean))
  ;[1 1 1] Toggle HUD key actions based on targeting - called from secondary fire GA
  (if (Variables|Default|GetShowHUDKeyActions)
    (if IsTargeting
      (WeaponHUDData|SetHudKeyActionsVisibility (Utilities|Array|MakeArray (Utilities|Struct|MakeWeaponHudKeyActionVisibility 1) (Utilities|Struct|MakeWeaponHudKeyActionVisibility 2 true)))
      (else
        (WeaponHUDData|SetHudKeyActionsVisibility (Utilities|Array|MakeArray (Utilities|Struct|MakeWeaponHudKeyActionVisibility 1 true) (Utilities|Struct|MakeWeaponHudKeyActionVisibility 2)))))))

(event FortWeapon|EventOnSpawnProjectile
  (params (SpawnProjectile :type Fort%20Projectile%20Base%20Object%20Reference))
  ;[1 1 1] On Spawn Projectile, set up for spawning gas trail
  (Class|FortSplatterSourceComponent|AddPreviousSource (Variables|Default|GetFortSplatterSource) SpawnProjectile))

(event FortWeapon|EventOnEquippedAndAttached
  ;[1 1 1] On equip
  (|CallParentFunction)
  (CallFunction|BindHolsterEvents)
  (CallFunction|OnEquip(Copy))
  (Networking|SwitchHasAuthority
    (:Authority
      (if (Variables|Default|GetEnableAmmoMod)
        (FortWeapon|ModAmmo 5)))))

(event Custom|ReplicateOnUnholstered
  ;[1 1 1] On equip
  (Networking|SwitchHasAuthority
    (:Authority
      (if (Variables|Default|GetEnableAmmoMod)
        (FortWeapon|ModAmmo 5)))))

(event Custom|OnUnholstered
  ;[1 1 1] On equip
  (Networking|SwitchHasAuthority
    (:Remote
      (CallFunction|ReplicateonUnholstered)
      (CallFunction|OnEquip(Copy)))))

(event Custom|UpdateAndAttachBGA
  (bind _attachedbgaactor (Variables|Default|GetAttachedBGAActor))
  (bind _attachedbgaactor_2 (Variables|Default|GetAttachedBGAActor))
  ;[1 1 1] Update and attach BGA
  (Utilities|IsValid _attachedbgaactor_2
    (:Is%20Valid
      (Game|Components|ProjectileMovement|StopSimulating (Variables|/Game/Athena/Items/HeldObjects/HeldObjectVersion2/BGA_HeldObject_Parent.BGA_HeldObject_Parent_C|GetFortProjectileMovement (Variables|Default|GetAttachedBGAActor)) (Collision|MakeHitResult false false 0.0 0.0 (Transformation|GetActorLocation self)))
      (Collision|SetIgnoreInstigatorCollision (Variables|Default|GetAttachedBGAActor) true :!class /Script/FortniteGame.BuildingGameplayActor)
      (Collision|SetCollisionEnabled (Variables|/Game/Athena/Items/Weapons/Prototype/PetrolPump/BGA_Petrol_Pickup.BGA_Petrol_Pickup_C|GetPetrolCan _attachedbgaactor) "NoCollision")
      (Collision|SetCollisionProfileName (Variables|/Game/Athena/Items/HeldObjects/HeldObjectVersion2/BGA_HeldObject_Parent.BGA_HeldObject_Parent_C|GetCapsule _attachedbgaactor) "FortTriggerOnlyWeapons" true :!class /Script/Engine.PrimitiveComponent)
      (Development|SetHiddeninGame (Variables|/Game/Athena/Items/Weapons/Prototype/PetrolPump/BGA_Petrol_Pickup.BGA_Petrol_Pickup_C|GetPetrolCan (Variables|Default|GetAttachedBGAActor)) true false)
      (Transformation|AttachActorToActor (Variables|Default|GetAttachedBGAActor) self (Variables|Default|GetAttachSocket) "SnapToTarget" "SnapToTarget" "SnapToTarget" true)
      (Transformation|SetActorRelativeTransform (Variables|Default|GetAttachedBGAActor) (Variables|Default|GetAttachedBGARelativeTransform) false true)
      (Variables|/Game/Athena/Items/HeldObjects/HeldObjectVersion2/BGA_HeldObject_Parent.BGA_HeldObject_Parent_C|SetRepStartMoving 0 (Variables|Default|GetAttachedBGAActor)))
    (:Is%20Not%20Valid
      (Actor|DestroyActor _attachedbgaactor_2))))

(event FortWeapon|EventOnLocalUnEquip
  ;[1 1 1] On unequip
  (|CallParentFunction)
  (CallFunction|UnbindHolsterEvents)
  (CallFunction|OnUnEquip(Copy))
  (Utilities|IsValid (Variables|Default|GetAttachedBGAActor)
    (:Is%20Valid
      (if (not (Variables|/Game/Athena/Items/Weapons/Prototype/PetrolPump/BGA_Petrol_Pickup.BGA_Petrol_Pickup_C|GetDropped (Variables|Default|GetAttachedBGAActor)))
        (Variables|/Game/Athena/Items/Weapons/Prototype/PetrolPump/BGA_Petrol_Pickup.BGA_Petrol_Pickup_C|SetShouldExplode true (Variables|Default|GetAttachedBGAActor))
        (Class|BGAPetrolPickup|HideandDestroy (Variables|Default|GetAttachedBGAActor))))))

(event Custom|ReplicateOnHolstered
  ;[1 1 1] On unequip
  (CallFunction|OnUnEquip(Copy))
  (Utilities|IsValid (Variables|Default|GetAttachedBGAActor)
    (:Is%20Valid
      (if (not (Variables|/Game/Athena/Items/Weapons/Prototype/PetrolPump/BGA_Petrol_Pickup.BGA_Petrol_Pickup_C|GetDropped (Variables|Default|GetAttachedBGAActor)))
        (Variables|/Game/Athena/Items/Weapons/Prototype/PetrolPump/BGA_Petrol_Pickup.BGA_Petrol_Pickup_C|SetShouldExplode true (Variables|Default|GetAttachedBGAActor))
        (Class|BGAPetrolPickup|HideandDestroy (Variables|Default|GetAttachedBGAActor))))))

(event Custom|OnHolstered
  ;[1 1 1] On unequip
  (Networking|SwitchHasAuthority
    (:Remote
      (CallFunction|ReplicateonHolstered)
      (CallFunction|OnUnEquip(Copy))
      (Utilities|IsValid (Variables|Default|GetAttachedBGAActor)
        (:Is%20Valid
          (if (not (Variables|/Game/Athena/Items/Weapons/Prototype/PetrolPump/BGA_Petrol_Pickup.BGA_Petrol_Pickup_C|GetDropped (Variables|Default|GetAttachedBGAActor)))
            (Variables|/Game/Athena/Items/Weapons/Prototype/PetrolPump/BGA_Petrol_Pickup.BGA_Petrol_Pickup_C|SetShouldExplode true (Variables|Default|GetAttachedBGAActor))
            (Class|BGAPetrolPickup|HideandDestroy (Variables|Default|GetAttachedBGAActor))))))))

(event Custom|OnEquip(Copy)
  ;[1 1 1] FX  Fuel trail
  (bind _returnvalue (Niagara|SpawnSystemAttached "/Game/Weapons/Prototype/Petrol/FX/NS_Petrol_FuelLeaking.NS_Petrol_FuelLeaking" (Variables|Visual|GetWeaponMesh) "Root" "0, 0, 0" "0, 0, 0" "KeepRelativeOffset" false false))
  (Variables|Default|SetNS_FuelLeaking _returnvalue))

(event Custom|OnUnEquip(Copy)
  ;[1 1 1] FX  Fuel trail
  (Components|DestroyComponent (Variables|Default|GetNS_FuelLeaking)))

(event EventOnPlayWeaponFireFX
  (params (bPersistentFire :type Boolean) (bSecondaryFire :type Boolean))
  ;[1 1 1] FX  Fuel trail
  (|CallParentFunction false false)
  (Components|Activation|Activate (Variables|Default|GetNS_FuelLeaking) false)
  (Audio|Components|Audio|FadeIn (Variables|Default|GetPourLoopSound) 0.2 1.0 0.0 "Linear" :!class /Script/Engine.AudioComponent)
  (Audio|Components|Audio|FadeOut (Variables|Default|GetPourStopSound) 1.0 0.0 "Linear" :!class /Script/Engine.AudioComponent)
  (SoundIndicator|StartSoundIndicatorLooping (Variables|SoundIndicator|GetSoundIndicatorComponent) (Transformation|GetActorLocation) "/Game/Sounds/Fort_Gadgets/GasPumpsCans/GasCan_Pour_Loop_Cue.GasCan_Pour_Loop_Cue" 1.0 0))

(event EventOnStopWeaponFireFX
  ;[1 1 1] FX  Fuel trail
  (|CallParentFunction)
  (Components|Activation|Deactivate (Variables|Default|GetNS_FuelLeaking))
  (Audio|Components|Audio|FadeOut (Variables|Default|GetPourLoopSound) 0.5 0.0 "Linear" :!class /Script/Engine.AudioComponent)
  (Audio|Components|Audio|FadeIn (Variables|Default|GetPourStopSound) 0.5 1.0 0.0 "Linear" :!class /Script/Engine.AudioComponent)
  (SoundIndicator|StopSoundIndicatorLooping (Variables|SoundIndicator|GetSoundIndicatorComponent)))

(fn OnRep_AttachedBGAActor
  (CallFunction|UpdateandAttachBGA))

(fn ConstructionScript)
