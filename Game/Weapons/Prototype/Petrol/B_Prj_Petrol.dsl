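; Petrol weapon projectile that refuels vehicles and petrol pickups, starting at the weapon muzzle with fuel from an override or the shot's ammo cost.
; On hit, it refuels pickups directly, zeroes its fuel for NoRefuel-tagged vehicles, or explodes on other non-fuel actors.
; Fuel-system vehicle hits send authority-only quest stats and refuel gameplay events, with a tag for filling to full.
; Explosions refuel the closest eligible pickup among the hit actors.
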(event Collision|EventHit
  (params (MyComp :type Primitive%20Component%20Object%20Reference) (Other :type Actor%20Object%20Reference) (OtherComp :type Primitive%20Component%20Object%20Reference) (bSelfMoved :type Boolean) (HitLocation :type Vector :default "0, 0, 0") (HitNormal :type Vector :default "0, 0, 0") (NormalImpulse :type Vector :default "0, 0, 0") (Hit :type Hit%20Result%20Structure%20%28by%20ref%29))
  (bind _returnvalue (Game|GetInstigator))
  (bind _returnvalue_2 (Fortnite|GetFortPlayerControllerfromActor _returnvalue))
  (bind _fueltoaddtopetrolpickup (Variables|Default|GetFuelToAddToPetrolPickup))
  ;[1 1 1] On hit, check for vehicles to refill, else explode (check for other refillable actors)
  (|CallParentFunction MyComp Other OtherComp bSelfMoved HitLocation HitNormal NormalImpulse Hit)
  (bind _returnvalue_3 (Vehicle|Fuel|UsesVehicleFuelSystem(Message) Other))
  (bind _asfort_athena_vehicle (Utilities|Casting|CastToFortAthenaVehicle Other)
    (:then
      (CallFunction|CheckifNoFuelTagPresent _asfort_athena_vehicle))
    (:CastFailed
      _))
  (if (not _returnvalue_3)
    (bind _asbga_petrol_pickup (Utilities|Casting|CastToBGA_Petrol_Pickup Other)
      (:then
        (Class|BGAPetrolPickup|Refuel _asbga_petrol_pickup (Variables|Default|GetFuelToAddToPetrolPickup) (Fortnite|GetFortPlayerControllerfromActor (Game|GetInstigator)) 0))
      (:CastFailed
        (Delivery|DoExplosionatLocation (Transformation|GetActorLocation self))))
    (else
      ;[1 1 1] Quest Update For if we hit vehicles, we should have this be cached and do it once for each different vehicle we refueld. Shouldn't do the work of SendStatEvent() over and over on the end life of a projectile, should happen from the vehicle or something else.
      (Networking|SwitchHasAuthority
        (:Authority
          (bind _returnvalue_4 (Class|FortVehicleInterface|GetGameplayTags(Message) Other))
          (Quest|SendCustomStatEventwithTags :self (Quests|GetQuestManager _returnvalue_2 "Invalid") :Type "Refuel" :TargetTags _returnvalue_4 :Count _fueltoaddtopetrolpickup)
          (CallFunction|SendGameplayEvents _fueltoaddtopetrolpickup _returnvalue_4 _returnvalue_2 Other))))))

(event EventBeginPlay
  (bind _overridefueltoaddtopetrolpickup (Variables|Default|GetOverrideFuelToAddToPetrolPickup))
  (bind _returnvalue (ScalableFloat|GetValueatLevel _overridefueltoaddtopetrolpickup))
  ;[1 1 1] On Begin Play, set Fuel to Add to Petrol Pickup
  (|CallParentFunction)
  (bind _returnvalue_2 (CallFunction|GetOwnerWeapon))
  (Transformation|Teleport self (Transformation|GetSocketLocation (Variables|/Script/FortniteGame.FortWeapon|GetWeaponMesh _returnvalue_2) "Muzzle") (Transformation|GetActorRotation self))
  (if (>= _returnvalue 0.0)
    (Variables|Default|SetFuelToAddToPetrolPickup (Math|Float|Round _returnvalue))
    (else
      (bind _returnvalue_3 (CallFunction|GetOwnerWeapon))
      (Utilities|IsValid _returnvalue_3
        (:Is%20Valid
          (Variables|Default|SetFuelToAddToPetrolPickup (FortWeapon|GetAmmoCostPerFire _returnvalue_3 (Variables|FortProjectileBase|GetChargePercent))))
        (:Is%20Not%20Valid
          (Variables|Default|SetFuelToAddToPetrolPickup 0))))))

(event EventOnStop
  (params (Hit :type Hit%20Result%20Structure%20%28by%20ref%29))
  ;[1 1 1] On Stop
  (|CallParentFunction Hit)
  (Rendering|SetVisibility (Variables|Default|GetStaticMesh) false true))

(event EventOnExploded
  (params (HitActors :type Array%20of%20Actor%20Object%20References) (HitResults :type Array%20of%20Hit%20Result%20Structures))
  (bind _self self)
  ;[1 1 1] On Exploded, refill nearby petrol pickup BGAs
  (|CallParentFunction HitActors HitResults)
  (Variables|Default|SetClosestPetrolPickupBGA (Variables|Default|GetNullBGA))
  (Variables|Default|SetShortestDistance -1.0)
  (Variables|Default|SetClosestPetrolPickupBGAIndex -1)
  (bind _array_element (Utilities|Array|ForEachLoopwithBreak HitActors)
    (:LoopBody
      (bind _asbga_petrol_pickup (Utilities|Casting|CastToBGA_Petrol_Pickup _array_element))
      (if (or (< (Variables|Default|GetShortestDistance) 0.0) (< (Class|AbilityFunctions|GetDistanceBetweenActors _asbga_petrol_pickup _self) (Variables|Default|GetShortestDistance)))
        (Utilities|IsValid (Actor|GetAttachParentActor _asbga_petrol_pickup)
          (:Is%20Valid
            (bind _asb_petrol_weapon (Utilities|Casting|CastToB_PetrolWeapon (Actor|GetAttachParentActor _asbga_petrol_pickup))
              (:then
                (if (Utilities|NotEqual(Object) (Game|GetInstigator _asb_petrol_weapon) (Game|GetInstigator))
                  _))
              (:CastFailed
                _))
            (Variables|Default|SetShortestDistance (Class|AbilityFunctions|GetDistanceBetweenActors _asbga_petrol_pickup _self))
            (Variables|Default|SetClosestPetrolPickupBGA _asbga_petrol_pickup))
          (:Is%20Not%20Valid
            _)) ; Are we attached to something?
      ))
    (:Completed
      (Utilities|IsValid (Variables|Default|GetClosestPetrolPickupBGA)
        (:Is%20Valid
          (Class|BGAPetrolPickup|Refuel (Variables|Default|GetClosestPetrolPickupBGA) (Variables|Default|GetFuelToAddToPetrolPickup) (Fortnite|GetFortPlayerControllerfromActor (Game|GetInstigator)) 0))))))

(event EventDestroyed
  ;[1 1 1] On Destroyed
  (|CallParentFunction))

(fn CheckIfNoFuelTagPresent
  (params (Vehicle%20Actor :type Fort%20Athena%20Vehicle%20Object%20Reference))
  (if (GameplayTags|HasMatchingGameplayTag Vehicle%20Actor "(TagName=\"Gameplay.Vehicle.NoRefuel\")")
    (Variables|Config|SetFuelAmount 0.0)
    (else
      (Variables|Config|SetFuelAmount (ScalableFloat|GetValueatLevel (Variables|Default|GetHotfix_RefillVehicleAmount))))))

(fn SendGameplayEvents
  (params (FuelToAdd :type Integer) (VehicleTags :type Gameplay%20Tag%20Container%20Structure) (Playercontroller :type Fort%20Player%20Controller%20Object%20Reference) (VehicleInterface :type Actor%20Object%20Reference))
  (bind _returnvalue (Math|Conversions|ToFloat(Integer) FuelToAdd))
  (bind _returnvalue_2 (Vehicle|Fuel|GetFuelCapacity(Message) VehicleInterface))
  (bind _returnvalue_3 (Vehicle|Fuel|GetFuel(Message) VehicleInterface))
  (if (< _returnvalue_3 _returnvalue_2)
    (bind _returnvalue_4 (CallFunction|GetOwnerWeapon))
    (bind _returnvalue_5 (GameplayTags|GetOwnedGameplayTags _returnvalue_4 :!class /Script/GameplayTags.GameplayTagAssetInterface))
    (if (>= (+ _returnvalue_3 _returnvalue) _returnvalue_2)
      (GameplayTags|AddGameplayTag VehicleTags "(TagName=\"Event.Vehicle.Refuel.ToFull\")" :!class /Script/GameplayTags.BlueprintGameplayTagLibrary)
      (else
        _))
    (Ability|SendGameplayEventtoActor Playercontroller "(TagName=\"Event.Vehicle.Refuel\")" (Utilities|Struct|MakeGameplayEventData "(TagName=\"\")" Playercontroller VehicleInterface 0 0 "()" _returnvalue_5 VehicleTags _returnvalue))))

(fn ConstructionScript
  (|CallParentFunction)
  (Variables|Config|SetFuelAmount (ScalableFloat|GetValueatLevel (Variables|Default|GetHotfix_RefillVehicleAmount))))
