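# Translucent lit material for colored petrol splash particles, used on Niagara sprites for a prototype weapon.
# A panning distortion normal, varied by a per-particle random dynamic parameter, warps the splash shape and normal UVs.
# Shape channels drive a two-color tint and contrast-adjusted opacity, which fades near the camera and at depth intersections.
# View-space normals sample a reflection texture matcap-style, adding tinted fake reflections to the emissive.
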Material /Game/Weapons/Prototype/Petrol/FX/M_Petrol_Splash_Color
graph {
 properties {
  bUsedWithNanite=true
  bUsedWithNiagaraSprites=true
  blendMode=BLEND_Translucent
  pixelDepthOffsetMode=PDOM_Legacy
  usedShadingModels=Default Lit
 }
 nodes [
  n1 = TextureCoordinate()
  n2 = Constant() {r=0.5}
  Add(A=Multiply(A=Subtract(A=@n1, B=@n2), B=ScalarParameter() {defaultValue=1.5, parameterName=Scale}), B=@n2)
  n3 = ParticleColor()
  n4 = TextureCoordinate()
  n5 = ScalarParameter() {defaultValue=0.2, parameterName=DistTile}
  n6 = DynamicParameter() {paramNames [Rand 0-1, Param2, Param3, Param4]}
  n7 = Time() {period=128}
  n8 = TextureSample(UVs=Panner(Coordinate=Multiply(A=@n4, B=@n5), Time=Add(A=@n6.Rand 0-1, B=@n7)) {speedX=0.05, speedY=0.25}) {samplerType=SAMPLERTYPE_Normal, texture=&T_Weap_Distort_N_02}
  n9 = LinearInterpolate(Alpha=@n6.Rand 0-1) {constA=0.5, constB=1.2}
  n10 = ScalarParameter() {defaultValue=0.1, parameterName=DistAmnt}
  n11 = Add(A=@n1, B=Multiply(A=ComponentMask(None=@n8.RGB) {g=true, r=true}, B=Multiply(A=@n9, B=@n10)))
  n12 = TextureSampleParameter2D(UVs=@n11) {parameterName=Shape, samplerSource=SSM_Clamp_WorldGroupSettings, texture=/Game/Effects/Fort_Effects/Textures/Water/T_water_splash_04}
  n13 = Saturate(None=MaterialFunctionCall(
    In=ChannelMaskParameter(None=@n12.RGBA) {maskChannel=Alpha, parameterName=Opacity Channel, defaultValue=<rgba 0 0 0 1>}.R
    Contrast=ScalarParameter() {defaultValue=1, parameterName=MaskContr}
   ) {materialFunction=&CheapContrast}.Result)
  n14 = Power(
   Base=MaterialFunctionCall(Distance=ScalarParameter() {defaultValue=90, parameterName=NearFade}) {materialFunction=&Near_Camera_Fade}.Result
   Exp=ScalarParameter() {defaultValue=5, parameterName=NearFadeExp}
  )
  DepthFade(
   Opacity=Multiply(A=Multiply(A=@n3.A, B=@n13), B=Saturate(None=@n14))
   FadeDistance=ScalarParameter() {defaultValue=15, parameterName=DF_Dist}
  )
  n15 = Power(
   Base=ChannelMaskParameter(None=@n12.RGBA) {maskChannel=Blue, parameterName=Diffuse Channel, defaultValue=<rgba 0 0 1 0>}.R
   Exp=ScalarParameter() {defaultValue=0.8, parameterName=FresnelBExp}
  )
  n16 = LinearInterpolate(
   A=VectorParameter() {parameterName=ColorA, defaultValue=<rgba 0.1875 0.1239 0.0069 1>}.RGB
   B=VectorParameter() {parameterName=ColorB, defaultValue=<rgba 0.1927 0.0858 0.0148 1>}.RGB
   Alpha=@n15
  )
  n17 = Add(A=Multiply(A=OneMinus(None=@n15), B=@n16), B=@n16)
  n18 = TextureSampleParameter2D(UVs=@n11) {parameterName=Normal, samplerType=SAMPLERTYPE_Normal, texture=/Game/Effects/Fort_Effects/Textures/Water/T_water_splash_04_n}
  n19 = Constant() {r=0.5}
  n20 = Add(A=Multiply(A=Transform(None=@n18.RGB) {transformType=TRANSFORM_View}, B=@n19), B=@n19)
  n21 = MaterialFunctionCall(
   In=TextureSample(UVs=ComponentMask(None=@n20) {g=true, r=true}) {texture=/Game/VisualThreatManager/StormVisuals/LargeClouds/T_EyeballReflection2}.G
   Contrast=ScalarParameter() {parameterName=ReflectionContrast}
  ) {materialFunction=&CheapContrast}
  n22 = VectorParameter() {parameterName=ReflectionColor, defaultValue=<rgba 0.276 0.1053 0 1>}
  MaterialFunctionCall(Emissive=Add(A=Multiply(A=@n17, B=@n3.RGB), B=Multiply(A=@n21.Result, B=@n22.RGB))) {materialFunction=&MF_FX_EmissiveAdjustments}
 ]
}
